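# Material function that optionally layers animated Madam Moth flames onto incoming material attributes, patterned by an Estate wallpaper mask and optionally limited to layer 2.
# Panning noise and caustics on distorted world-aligned UVs drive a color-curve atlas lookup, with extra fire noise skipped on low/mobile.
# The wallpaper mask and Fresnel shape the emissive glow, which uses inverse eye adaptation and is added to the base emissive.
# Metallic and roughness also shift toward flame-specific values.
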MaterialFunction /Game/Environments/Helios/Sets/Estate/MaterialFunctions/MF_MadamMoth_Fire
graph {nodes [
  n1 = FunctionInput() {inputType=FunctionInput_MaterialAttributes}
  n2 = GetMaterialAttributes(None=@n1) {attributeGetTypes [
    #1
    #2
    #3
    #4
   ]}
  n3 = TextureSample() {samplerType=SAMPLERTYPE_Masks, texture=/Game/Environments/Helios/Sets/Estate/Textures/T_Estate_Wallpaper_A_A_M}
  n4 = WorldPosition()
  n5 = ScalarParameter() {defaultValue=220, group=Seasonal, parameterName=MadamMoth_TextureDivisor}
  n6 = NamedRerouteUsage()
  n7 = Add(
   A=MaterialFunctionCall!(UseLocalSpace, WorldPosition=Divide(A=@n4.XYZ, B=@n5), NormalOverride) {materialFunction=/Game/Environments/Apollo/Sets/Skyfire/Materials/MaterialFunctions/MF_WorldAlignedUVs_HardEdge}.Result
   B=Multiply(A=ComponentMask(None=@n6) {g=true, r=true}, B=ScalarParameter() {defaultValue=0.02, group=Seasonal, parameterName=DistortAmount})
  )
  n8 = TextureSample(UVs=Panner(Coordinate=@n7) {speedY=-0.0311}) {samplerType=SAMPLERTYPE_LinearGrayscale, texture=/Engine/Functions/Engine_MaterialFunctions02/ExampleContent/Textures/LowResBlurredNoise}
  n9 = Panner(Coordinate=@n7) {speedY=-0.0432}
  n10 = LinearInterpolate(
   A=ScalarParameter() {defaultValue=-0.05, parameterName=Disappear min}
   B=ScalarParameter() {defaultValue=1, parameterName=Disappear max}
   Alpha=Power(Base=TextureSample(UVs=@n9) {samplerType=SAMPLERTYPE_LinearGrayscale, texture=/Game/Global/Textures/Default/Caustics/T_Nosie_Caustics_Grayscale}.R) {constExponent=1.5}
  ) {constA=-2, constB=2}
  n11 = Add(A=Saturate(None=LinearInterpolate(Alpha=@n8.R) {constA=-0.02, constB=0.06}), B=Multiply(A=@n8.R, B=Saturate(None=@n10)))
  n12 = Multiply(A=@n7) {constB=0.8}
  n13 = Time()
  n14 = Multiply(A=Constant2Vector() {r=1}) {constB=0.01}
  n15 = TextureSample(UVs=Add(A=@n12, B=Multiply(A=@n13, B=@n14))) {samplerSource=SSM_Wrap_WorldGroupSettings, texture=/Game/Global/Textures/Default/Noise/T_FireNoise_01}
  n16 = Multiply(A=@n13, B=Multiply(A=@n14) {constB=-1})
  n17 = TextureSample(UVs=Multiply(A=Add(A=@n12, B=@n16)) {constB=1.1}) {samplerSource=SSM_Wrap_WorldGroupSettings, texture=&T_FireNoise_01}
  n18 = Multiply(A=Power(Base=Add(A=@n15.R, B=@n17.R)) {constExponent=3}) {constB=0.04}
  n19 = CurveAtlasRowParameter(CurveTime=MaterialFunctionCall!(Default=Add(A=@n11, B=@n18), Low/Mobile=@n11) {materialFunction=/Game/Effects/Fort_Effects/Materials/_Functions/MF_LowMobile_PathingSwitch}.Result) {
   atlas=/Game/Athena/LinearColorCurves/CAtlas_Contrail_MadameMoth_Color
   curve=/Game/Athena/LinearColorCurves/CV_Contrail_MadameMoth_Color4
   defaultValue=6
   parameterName=Color
  }
  n20 = MaterialFunctionCall() {materialFunction=&MF_Fresnel_FX}
  n21 = Multiply(
   A=LinearInterpolate(Alpha=NamedRerouteDeclaration(None=@n3.R) {name=FlamePattern, nodeColor=<rgba 0.9922 0 1 1>}) {constA=0.02}
   B=Multiply(A=@n19, B=LinearInterpolate(Alpha=@n20.Fresnel) {constA=0.2})
  )
  n22 = NamedRerouteUsage()
  n23 = SetMaterialAttributes!(
   MaterialAttributes=@n2.MaterialAttributes
   EmissiveColor=Add(A=@n2.EmissiveColor, B=EyeAdaptationInverse(LightValueInput=@n21).EyeAdaptationInverse)
   Metallic=LinearInterpolate(A=NamedRerouteUsage(), B=ScalarParameter() {defaultValue=0.65, group=Seasonal, parameterName=FlameMetallic}, Alpha=@n22)
   Roughness=LinearInterpolate(A=@n2.Roughness, B=ScalarParameter() {defaultValue=0.22, group=Seasonal, parameterName=FlameRoughness}, Alpha=@n22)
  ) {attributeSetTypes [#1, #3, #4]}
  n24 = Constant()
  n25 = Constant() {r=1}
  n26 = StaticSwitchParameter(
   True=ComponentMask(None=MaterialFunctionCall!(1=@n24, 2=@n25, 3, 4, UVOverride) {materialFunction=/Game/Environments/Asteria/Sets/NeonCity/Techart/MF_NeonCity_4Layer_Vec4}.Result) {r=true}
   False=Constant() {r=1}
  ) {group=Seasonal, parameterName=UseLayer2}
  FunctionOutput(None=StaticSwitchParameter(True=BlendMaterialAttributes(A=@n1, B=@n23, Alpha=@n26), False=@n1) {group=Seasonal, parameterName=UseMadamMothFlames})
  NamedRerouteDeclaration(None=@n2.Normal) {name=Normal.In, nodeColor=<rgba 0 1 0.3438 1>}
  NamedRerouteDeclaration(None=@n2.Metallic) {name=Metallic.In, nodeColor=<rgba 0 0.0859 1 1>}
 ]}
